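# Translucent material for the Fireplace cosmetic wheel's FX on skeletal meshes, drawing an animated, bright red-orange emissive glow.
# A vertically panned flow texture distorts scaled and offset UVs that sample a circular gradient, whose slightly reduced, clamped value becomes opacity.
# A pixel-versus-scene depth comparison keeps that opacity only where the pixel is less than one unit behind scene depth, and zeroes it otherwise.
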Material /VehicleCosmetics/Wheels/Fireplace/Materials/M_Wheel_FirePlace_FX
graph {
 properties {
  bAutomaticallySetUsageInEditor=false
  bUseMaterialAttributes=true
  bUsedWithMeshDeformer=true
  bUsedWithSkeletalMesh=true
  blendMode=BLEND_Translucent
  pixelDepthOffsetMode=PDOM_Legacy
  usedShadingModels=Default Lit
 }
 nodes [
  n1 = Subtract(A=PixelDepth().R, B=SceneDepth().R)
  n2 = Constant()
  n3 = TextureCoordinate()
  n4 = Constant2Vector() {g=-2, r=-0.5}
  n5 = Panner(Coordinate=TextureCoordinate() {uTiling=0.3, vTiling=0.3}, Time=Time()) {speedY=0.8}
  n6 = Add(
   A=Multiply(A=Constant2Vector() {g=0.2, r=0.2}, B=Add(A=@n3, B=@n4))
   B=ComponentMask(None=TextureSample(UVs=@n5) {samplerType=SAMPLERTYPE_LinearColor, texture=/VehicleCosmetics/Wheels/Fireplace/Textures/Flow_04}.RGB) {g=true, r=true}
  )
  n7 = Clamp(None=Subtract(A=TextureSample(UVs=@n6) {samplerType=SAMPLERTYPE_Grayscale, texture=/VehicleCosmetics/Wheels/_SharedTextures/FX/GradientCircle01}.G) {constB=0.05})
  MakeMaterialAttributes!(
   BaseColor
   Metallic
   Specular
   Roughness
   Anisotropy
   EmissiveColor=Constant4Vector() {constant=<rgba 10 1 0 0>}.RGBA
   Opacity=Clamp(None=If(A=@n1, "A > B"=@n2, "A == B"=@n2, "A < B"=@n7) {constB=1})
   OpacityMask
   Normal
   Tangent
   WorldPositionOffset
   SubsurfaceColor
   ClearCoat
   ClearCoatRoughness
   AmbientOcclusion
   Refraction
   CustomizedUVs_0
   CustomizedUVs_1
   CustomizedUVs_2
   CustomizedUVs_3
   CustomizedUVs_4
   CustomizedUVs_5
   CustomizedUVs_6
   CustomizedUVs_7
   PixelDepthOffset
   ShadingModel
   Displacement
  )
 ]
}
