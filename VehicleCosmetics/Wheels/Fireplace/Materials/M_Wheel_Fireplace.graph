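# Masked material for the Fireplace vehicle wheel cosmetic, adding an animated emissive glow to parts of the rim.
# MF_WheelParent_Simple supplies the base attributes, rim RGBA mask and vertex color.
# A clamped, rescaled sine scrolls along U at Speed, scales by EmissivePower and Base Color, and blends into Emissive through summed RGB rim channels times inverted vertex red.
# The result passes through vehicle damage and de-rez transition functions.
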Material /VehicleCosmetics/Wheels/Fireplace/Materials/M_Wheel_Fireplace
graph {
 properties {
  bAutomaticallySetUsageInEditor=false
  bUseMaterialAttributes=true
  bUsedWithMeshDeformer=true
  bUsedWithSkeletalMesh=true
  blendMode=BLEND_Masked
  pixelDepthOffsetMode=PDOM_Legacy
  usedShadingModels=Default Lit
 }
 nodes [
  n1 = MaterialFunctionCall() {materialFunction=&MF_WheelParent_Simple}
  n2 = GetMaterialAttributes(None=@n1.MaterialAttributes) {attributeGetTypes [#1, #2]}
  n3 = ScalarParameter() {defaultValue=1, group=5_Special Effects, parameterName=Speed}
  n4 = Add(A=ComponentMask(None=TextureCoordinate()) {r=true}, B=Multiply(A=Time(), B=Multiply(A=@n3) {constB=-1}))
  n5 = ConstantBiasScale(None=Clamp(None=Sine(None=@n4))) {bias=0.05, scale=6}
  n6 = ScalarParameter() {defaultValue=1, group=5_Special Effects, parameterName=EmissivePower, sortPriority=33}
  n7 = Add(A=ComponentMask(None=@n1.RimRGBAMask) {r=true}, B=ComponentMask(None=@n1.RimRGBAMask) {g=true})
  n8 = ComponentMask(None=@n1.RimRGBAMask) {b=true}
  n9 = LinearInterpolate(
   A=@n2.Emissive Color
   B=Multiply(A=@n2.Base Color, B=Multiply(A=@n5, B=@n6) {constB=0})
   Alpha=Multiply(A=Add(A=@n7, B=@n8), B=OneMinus(None=@n1.VertexColorR))
  )
  MaterialFunctionCall(In=MaterialFunctionCall(Material=SetMaterialAttributes(MaterialAttributes=@n2.MaterialAttributes, Emissive Color=@n9) {attributeSetTypes [#1]}) {materialFunction=&MF_BaseVehicle_Damage}.Result) {materialFunction=&MF_VehicleDeRez_Transition}
 ]
}
